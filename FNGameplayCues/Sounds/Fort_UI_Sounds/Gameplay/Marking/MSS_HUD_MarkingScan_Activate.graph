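# One-shot stereo MetaSound source for the marking scan activation sound in the HUD, classed as critical in-game UI audio.
# On play, a random float sets a slight pitch shift and triggers a wave player for the MarkingScan_Activate wave, which ends the sound when finished.
# It uses marking-scan attenuation, disables virtualization and overrides concurrency to allow two instances, stopping the oldest.
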MetaSoundSource /FNGameplayCues/Sounds/Fort_UI_Sounds/Gameplay/Marking/MSS_HUD_MarkingScan_Activate
inputs [UE.Source.OnPlay {type=Trigger}]
interfaces [UE.OutputFormat.Stereo:1.0, UE.Source.OneShot:1.0, UE.Source:1.1]
outputs [
 "UE.OutputFormat.Stereo.Audio:0" {type=Audio}
 "UE.OutputFormat.Stereo.Audio:1" {type=Audio}
 UE.Source.OneShot.OnFinished {type=Trigger}
]
pages [Default {
  input_defaults {UE.Source.OnPlay=null}
  nodes [
   n1 = UE.RandomFloat(Next=^UE.Source.OnPlay) {also {defaults {Max=1.2}}}
   WavePlayer_0 = UE.Wave Player:Stereo(Play=@n1.On Next, Pitch Shift=@n1.Value) {also {defaults {Wave Asset=&/MarkingScan_Activate}, version [1,1]}}
  ]
  links [
   {dst="@UE.Source.OneShot.OnFinished", src=WavePlayer_0.On Finished}
   {dst="@UE.OutputFormat.Stereo.Audio:0", src=WavePlayer_0.Out Left}
   {dst="@UE.OutputFormat.Stereo.Audio:1", src=WavePlayer_0.Out Right}
  ]
 }]
properties {
 assetImportData=&self:AssetImportData
 attenuationSettings=&/ATT_MarkingScan
 bOverrideConcurrency=true
 maxDistance=24000
 numChannels=2
 outputFormat=Stereo
 sampleRate=48000
 soundClassObject=&Fort_UI_InGame_Critical
 virtualizationMode=Disabled
 concurrencyOverrides {maxCount=2, resolutionRule=StopOldest}
}
